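# Physics asset for the male medium Floatilla Captain body's pants, setting up front and back dynamic cloth bones on each leg.
# Capsules on both thighs anchor tiny base capsules through constraints with free swing and twist.
# Each base carries a flat box body on a tightly limited swing and twist constraint, with the back pieces allowed more swing than the front.
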PhysicsAsset /BRCosmetics/Characters/Player/Male/Medium/Bodies/M_MED_Floatilla_Captain/Meshes/M_MED_Floatilla_Captain_Pants_Physics
bodies {
 dyn_pants_bk_base_l [{shapeType=Capsule, length=1, radius=1}]
 dyn_pants_bk_base_r [{shapeType=Capsule, length=1, radius=1}]
 dyn_pants_bk_l [{shapeType=Box, extentX=13, extentY=3, extentZ=13, center=<6 0 0>}]
 dyn_pants_bk_r [{shapeType=Box, extentX=13, extentY=3, extentZ=13, center=<-6 0 0>, rotation=<rot 0 0 180>}]
 dyn_pants_fr_base_l [{shapeType=Capsule, length=1, radius=1}]
 dyn_pants_fr_base_r [{shapeType=Capsule, length=1, radius=1}]
 dyn_pants_fr_l [{shapeType=Box, extentX=13, extentY=3, extentZ=13, center=<6 0 0>}]
 dyn_pants_fr_r [{shapeType=Box, extentX=13, extentY=3, extentZ=13, center=<-6 0 0>, rotation=<rot 0 0 180>}]
 thigh_l [{shapeType=Capsule, length=33, radius=6.25, center=<-19.889 -0.0002 1.1247>, rotation=<rot 88.5609 -180 -180>}]
 thigh_r [{shapeType=Capsule, length=33, radius=6.25, center=<19.889 0.0002 -1.1247>, rotation=<rot 88.5598 -180 0>}]
}
constraints [
 {
  bone1Name=dyn_pants_fr_base_l
  bone2Name=thigh_l
  swing1LimitDegrees=45
  swing1Motion=Free
  swing2LimitDegrees=45
  swing2Motion=Free
  twistLimitDegrees=45
  twistMotion=Free
 }
 {
  bone1Name=dyn_pants_bk_base_l
  bone2Name=thigh_l
  swing1LimitDegrees=45
  swing1Motion=Free
  swing2LimitDegrees=45
  swing2Motion=Free
  twistLimitDegrees=45
  twistMotion=Free
 }
 {bone1Name=dyn_pants_fr_l, bone2Name=dyn_pants_fr_base_l, swing1LimitDegrees=20, swing2LimitDegrees=10, twistLimitDegrees=5}
 {bone1Name=dyn_pants_bk_l, bone2Name=dyn_pants_bk_base_l, swing1LimitDegrees=30, swing2LimitDegrees=10, twistLimitDegrees=5}
 {bone1Name=dyn_pants_fr_r, bone2Name=dyn_pants_fr_base_r, swing1LimitDegrees=20, swing2LimitDegrees=10, twistLimitDegrees=5}
 {bone1Name=dyn_pants_bk_r, bone2Name=dyn_pants_bk_base_r, swing1LimitDegrees=30, swing2LimitDegrees=10, twistLimitDegrees=5}
 {
  bone1Name=dyn_pants_fr_base_r
  bone2Name=thigh_r
  swing1LimitDegrees=45
  swing1Motion=Free
  swing2LimitDegrees=45
  swing2Motion=Free
  twistLimitDegrees=45
  twistMotion=Free
 }
 {
  bone1Name=dyn_pants_bk_base_r
  bone2Name=thigh_r
  swing1LimitDegrees=45
  swing1Motion=Free
  swing2LimitDegrees=45
  swing2Motion=Free
  twistLimitDegrees=45
  twistMotion=Free
 }
]
