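# Physics asset for the Floatilla Captain medium male player body, covering its dangling accessories: two chest bags, six side pouches and three spikes.
# Each accessory is a box body constrained to a tiny capsule on its base bone, with narrow angular limits and some axes locked per piece.
# Spine and thigh capsules supply body volumes, and only the spike bases are constrained to the spine, with all angular axes free.
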PhysicsAsset /BRCosmetics/Characters/Player/Male/Medium/Bodies/M_MED_Floatilla_Captain/Meshes/M_MED_Floatilla_Captain_Physics
bodies {
 dyn_chest_bag_a [{shapeType=Box, extentX=7.5, extentY=2, extentZ=8, center=<3.5993 0 0>}]
 dyn_chest_bag_a_base [{shapeType=Capsule, length=1, radius=1}]
 dyn_chest_bag_b [{shapeType=Box, extentX=7.5, extentY=2, extentZ=8, center=<4 0 0>}]
 dyn_chest_bag_b_base [{shapeType=Capsule, length=1, radius=1}]
 dyn_pouch_a_base_l [{shapeType=Capsule, length=1, radius=1}]
 dyn_pouch_a_base_r [{shapeType=Capsule, length=1, radius=1}]
 dyn_pouch_a_l [{shapeType=Box, extentX=5.7846, extentY=3.501, extentZ=11.225, center=<3.6407 -0.4026 1.9054>, rotation=<rot 0 -94.134 -89.4607>}]
 dyn_pouch_a_r [{shapeType=Box, extentX=5.7846, extentY=3.501, extentZ=11.225, center=<-3.6407 0.4026 -1.9054>, rotation=<rot 0 -94.134 90.5394>}]
 dyn_pouch_b_base_l [{shapeType=Capsule, length=1, radius=1}]
 dyn_pouch_b_base_r [{shapeType=Capsule, length=1, radius=1}]
 dyn_pouch_b_l [{shapeType=Box, extentX=5.8317, extentY=3.7596, extentZ=11.3064, center=<4.0287 -0.5197 1.9445>, rotation=<rot 0 85.1679 -88.7493>}]
 dyn_pouch_b_r [{shapeType=Box, extentX=5.8317, extentY=3.7596, extentZ=11.3064, center=<-4.0287 0.5197 -1.9445>, rotation=<rot 0 85.1679 91.2507>}]
 dyn_pouch_c_base_l [{shapeType=Capsule, length=1, radius=1}]
 dyn_pouch_c_base_r [{shapeType=Capsule, length=1, radius=1}]
 dyn_pouch_c_l [{shapeType=Box, extentX=5.804, extentY=3.6034, extentZ=11.2907, center=<4.4386 -0.3326 2.4756>, rotation=<rot 0 87.704 -83.9376>}]
 dyn_pouch_c_r [{shapeType=Box, extentX=5.804, extentY=3.6034, extentZ=11.2907, center=<-4.4386 0.3326 -2.4756>, rotation=<rot 0 87.704 96.0624>}]
 dyn_spike [{shapeType=Box, extentX=6, extentY=1, extentZ=3, center=<2.9189 -0.0001 0>}]
 dyn_spike_base [{shapeType=Capsule, length=1, radius=1}]
 dyn_spike_base_l [{shapeType=Capsule, length=1, radius=1}]
 dyn_spike_base_r [{shapeType=Capsule, length=1, radius=1, rotation=<rot 0 0 180>}]
 dyn_spike_l [{shapeType=Box, extentX=4, extentY=1, extentZ=3, center=<1.9193 -0.0002 0>}]
 dyn_spike_r [{shapeType=Box, extentX=4, extentY=1, extentZ=3, center=<-1.9193 0.0002 0>, rotation=<rot 0 0 180>}]
 spine_05 [{shapeType=Capsule, length=12.75, radius=11.25, center=<-2.168 -1.0425 0>}]
 thigh_l [{shapeType=Capsule, length=30.75, radius=8.5, center=<-19.889 -0.0002 1.1247>, rotation=<rot 88.5609 -180 -180>}]
 thigh_r [{shapeType=Capsule, length=30.75, radius=8.5, center=<19.889 0.0002 -1.1247>, rotation=<rot 88.5598 -180 0>}]
}
constraints [
 {
  bone1Name=dyn_pouch_a_l
  bone2Name=dyn_pouch_a_base_l
  swing1LimitDegrees=2
  swing1Motion=Locked
  swing2LimitDegrees=2
  swing2Motion=Locked
  twistLimitDegrees=10
 }
 {bone1Name=dyn_pouch_b_l, bone2Name=dyn_pouch_b_base_l, swing1LimitDegrees=10, swing2LimitDegrees=10, twistLimitDegrees=10}
 {bone1Name=dyn_pouch_c_l, bone2Name=dyn_pouch_c_base_l, swing1LimitDegrees=5, swing2LimitDegrees=10, twistLimitDegrees=2, twistMotion=Locked}
 {
  bone1Name=dyn_spike
  bone2Name=dyn_spike_base
  swing1LimitDegrees=10
  swing2LimitDegrees=2
  swing2Motion=Locked
  twistLimitDegrees=45
  twistMotion=Locked
 }
 {
  bone1Name=dyn_spike_l
  bone2Name=dyn_spike_base_l
  swing1LimitDegrees=10
  swing2LimitDegrees=2
  swing2Motion=Locked
  twistLimitDegrees=45
  twistMotion=Locked
 }
 {
  bone1Name=dyn_spike_base_l
  bone2Name=spine_05
  swing1LimitDegrees=45
  swing1Motion=Free
  swing2LimitDegrees=45
  swing2Motion=Free
  twistLimitDegrees=45
  twistMotion=Free
 }
 {
  bone1Name=dyn_spike_base
  bone2Name=spine_05
  swing1LimitDegrees=45
  swing1Motion=Free
  swing2LimitDegrees=45
  swing2Motion=Free
  twistLimitDegrees=45
  twistMotion=Free
 }
 {
  bone1Name=dyn_spike_base_r
  bone2Name=spine_05
  swing1LimitDegrees=45
  swing1Motion=Free
  swing2LimitDegrees=45
  swing2Motion=Free
  twistLimitDegrees=45
  twistMotion=Free
 }
 {
  bone1Name=dyn_spike_r
  bone2Name=dyn_spike_base_r
  swing1LimitDegrees=5
  swing2LimitDegrees=2
  swing2Motion=Locked
  twistLimitDegrees=45
  twistMotion=Locked
 }
 {
  bone1Name=dyn_pouch_a_r
  bone2Name=dyn_pouch_a_base_r
  swing1LimitDegrees=2
  swing1Motion=Locked
  swing2LimitDegrees=2
  swing2Motion=Locked
  twistLimitDegrees=10
 }
 {bone1Name=dyn_pouch_b_r, bone2Name=dyn_pouch_b_base_r, swing1LimitDegrees=10, swing2LimitDegrees=10, twistLimitDegrees=10}
 {bone1Name=dyn_pouch_c_r, bone2Name=dyn_pouch_c_base_r, swing1LimitDegrees=5, swing2LimitDegrees=10, twistLimitDegrees=2, twistMotion=Locked}
 {
  bone1Name=dyn_chest_bag_a
  bone2Name=dyn_chest_bag_a_base
  swing1LimitDegrees=15
  swing2LimitDegrees=5
  swing2Motion=Locked
  twistLimitDegrees=45
  twistMotion=Locked
 }
 {
  bone1Name=dyn_chest_bag_b
  bone2Name=dyn_chest_bag_b_base
  swing1LimitDegrees=5
  swing2LimitDegrees=5
  swing2Motion=Locked
  twistLimitDegrees=45
  twistMotion=Locked
 }
]
