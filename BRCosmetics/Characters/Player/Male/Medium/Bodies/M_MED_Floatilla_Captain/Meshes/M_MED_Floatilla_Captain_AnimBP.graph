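# Post-process style AnimBP for the Floatilla Captain male medium body, adding secondary motion to copied parent-mesh poses.
# Constraints anchor necklace, spike, zipper and bag bones; additive bone offsets correct thighs, necklace, pouches and pants using shoulder, leg and knee angles from EventGraph.
# AnimDynamics drives zippers, then two RigidBody passes simulate physics, with pants weight reduced when crouching, jumping or falling.
# Both shoulderUp_L and shoulderUp_R read clavicle_l.
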AnimBlueprint /BRCosmetics/Characters/Player/Male/Medium/Bodies/M_MED_Floatilla_Captain/Meshes/M_MED_Floatilla_Captain_AnimBP
target_skeleton=&/M_MED_Floatilla_Captain_Skeleton
graphs [
 AnimGraph {nodes [
   n1 = "Constraint - thigh_volume_l (R)"(ComponentPose=Local To Component(LocalPose=Copy Pose From Mesh() {bUseAttachedParent=true}), Alpha=0.1, ConstraintWeights_0=1) {
    alphaBoolBlend {blendOption=Linear}
    boneToModify {boneName=thigh_volume_l}
    constraintSetup [{transformType=Rotation, targetBone {boneName=thigh_l}}]
    constraintWeights [1]
   }
   n2 = "Transform (Modify) Bone - Bone: thigh_volume_l"(ComponentPose=@n1, Alpha=Get legUp_L() -> (legUp_L:double), Translation="-3,-3,0", Rotation="0,0,0", Scale="1.2,1.2,1.2") {
    rotationMode=BMM_Additive
    rotationSpace=BCS_BoneSpace
    scaleMode=BMM_Additive
    scaleSpace=BCS_BoneSpace
    translationMode=BMM_Additive
    translationSpace=BCS_BoneSpace
    alphaBoolBlend {blendOption=Linear}
    alphaScaleBiasClamp {bMapRange=true, inRange {max=0, min=-100}, outRange {max=0, min=1}}
    boneToModify {boneName=thigh_volume_l}
    scale=<1.2 1.2 1.2>
    translation=<-3 -3>
   }
   n3 = "Constraint - thigh_volume_r (R)"(ComponentPose=@n2, Alpha=0.1, ConstraintWeights_0=1) {
    alphaBoolBlend {blendOption=Linear}
    boneToModify {boneName=thigh_volume_r}
    constraintSetup [{transformType=Rotation, targetBone {boneName=thigh_r}}]
    constraintWeights [1]
   }
   n4 = "Transform (Modify) Bone - Bone: thigh_volume_r"(ComponentPose=@n3, Alpha=Get legUp_R() -> (legUp_R:double), Translation="3,3,0", Rotation="0,0,0", Scale="1.2,1.2,1.2") {
    rotationMode=BMM_Replace
    rotationSpace=BCS_BoneSpace
    scaleMode=BMM_Additive
    scaleSpace=BCS_BoneSpace
    translationMode=BMM_Additive
    translationSpace=BCS_BoneSpace
    alphaBoolBlend {blendOption=Linear}
    alphaScaleBiasClamp {bMapRange=true, inRange {max=0, min=-100}, outRange {max=0, min=1}}
    boneToModify {boneName=thigh_volume_r}
    scale=<1.2 1.2 1.2>
    translation=<3 3>
   }
   n5 = "Constraint - dyn_necklace_fr_base (PPPPPP)"!(
    ComponentPose=@n4
    Alpha=1
    ConstraintWeights_0=0.0668
    ConstraintWeights_1:float=0.1466
    ConstraintWeights_2:float=0.0814
    ConstraintWeights_3:float=0.2351
    ConstraintWeights_4:float=0.3987
    ConstraintWeights_5:float=0.0713
   ) -> (Pose) {
    alphaBoolBlend {blendOption=Linear}
    boneToModify {boneName=dyn_necklace_fr_base}
    constraintSetup [
     {targetBone {boneName=clavicle_l}}
     {targetBone {boneName=neck_01}}
     {targetBone {boneName=neck_02}}
     {targetBone {boneName=spine_04}}
     {targetBone {boneName=spine_05}}
     {targetBone {boneName=clavicle_r}}
    ]
    constraintWeights [1,1,1,1,1,1]
   }
   n6 = "Constraint - dyn_necklace_bk_base (PPPP)"!(
    ComponentPose=@n5
    Alpha=1
    ConstraintWeights_0=0.4125
    ConstraintWeights_1:float=0.5355
    ConstraintWeights_2:float=0.0366
    ConstraintWeights_3:float=0.0155
   ) -> (Pose) {
    alphaBoolBlend {blendOption=Linear}
    boneToModify {boneName=dyn_necklace_bk_base}
    constraintSetup [{targetBone {boneName=neck_01}}, {targetBone {boneName=neck_02}}, {targetBone {boneName=head}}, {targetBone {boneName=spine_05}}]
    constraintWeights [1,1,1,1]
   }
   n7 = "Constraint - dyn_necklace_fr_base_l (PPPPP)"!(
    ComponentPose=@n6
    Alpha=1
    ConstraintWeights_0=0.2775
    ConstraintWeights_1:float=0.1522
    ConstraintWeights_2:float=0.1075
    ConstraintWeights_3:float=0.098
    ConstraintWeights_4:float=0.3648
   ) -> (Pose) {
    alphaBoolBlend {blendOption=Linear}
    boneToModify {boneName=dyn_necklace_fr_base_l}
    constraintSetup [
     {targetBone {boneName=clavicle_l}}
     {targetBone {boneName=neck_01}}
     {targetBone {boneName=neck_02}}
     {targetBone {boneName=spine_04}}
     {targetBone {boneName=spine_05}}
    ]
    constraintWeights [1,1,1,1,1]
   }
   n8 = "Constraint - dyn_necklace_side_base_l (PPPPP)"!(
    ComponentPose=@n7
    Alpha=1
    ConstraintWeights_0=0.3978
    ConstraintWeights_1:float=0.1422
    ConstraintWeights_2:float=0.0987
    ConstraintWeights_3:float=0.0256
    ConstraintWeights_4:float=0.3358
   ) -> (Pose) {
    alphaBoolBlend {blendOption=Linear}
    boneToModify {boneName=dyn_necklace_side_base_l}
    constraintSetup [
     {targetBone {boneName=clavicle_l}}
     {targetBone {boneName=neck_01}}
     {targetBone {boneName=neck_02}}
     {targetBone {boneName=spine_04}}
     {targetBone {boneName=spine_05}}
    ]
    constraintWeights [1,1,1,1,1]
   }
   n9 = "Constraint - dyn_necklace_bk_base_l (PPPPP)"!(
    ComponentPose=@n8
    Alpha=1
    ConstraintWeights_0=0.0943
    ConstraintWeights_1:float=0.3603
    ConstraintWeights_2:float=0.4022
    ConstraintWeights_3:float=0.0194
    ConstraintWeights_4:float=0.1238
   ) -> (Pose) {
    alphaBoolBlend {blendOption=Linear}
    boneToModify {boneName=dyn_necklace_bk_base_l}
    constraintSetup [
     {targetBone {boneName=clavicle_l}}
     {targetBone {boneName=neck_01}}
     {targetBone {boneName=neck_02}}
     {targetBone {boneName=head}}
     {targetBone {boneName=spine_05}}
    ]
    constraintWeights [1,1,1,1,1]
   }
   n10 = "Constraint - dyn_necklace_fr_base_r (PPPPP)"!(
    ComponentPose=@n9
    Alpha=1
    ConstraintWeights_0=0.1547
    ConstraintWeights_1:float=0.1092
    ConstraintWeights_2:float=0.2756
    ConstraintWeights_3:float=0.0983
    ConstraintWeights_4:float=0.3622
   ) -> (Pose) {
    alphaBoolBlend {blendOption=Linear}
    boneToModify {boneName=dyn_necklace_fr_base_r}
    constraintSetup [
     {targetBone {boneName=neck_01}}
     {targetBone {boneName=neck_02}}
     {targetBone {boneName=clavicle_r}}
     {targetBone {boneName=spine_04}}
     {targetBone {boneName=spine_05}}
    ]
    constraintWeights [1,1,1,1,1]
   }
   n11 = "Constraint - dyn_necklace_side_base_r (PPPPP)"!(
    ComponentPose=@n10
    Alpha=1
    ConstraintWeights_0=0.1378
    ConstraintWeights_1:float=0.0966
    ConstraintWeights_2:float=0.4014
    ConstraintWeights_3:float=0.0262
    ConstraintWeights_4:float=0.338
   ) -> (Pose) {
    alphaBoolBlend {blendOption=Linear}
    boneToModify {boneName=dyn_necklace_side_base_r}
    constraintSetup [
     {targetBone {boneName=neck_01}}
     {targetBone {boneName=neck_02}}
     {targetBone {boneName=clavicle_r}}
     {targetBone {boneName=spine_04}}
     {targetBone {boneName=spine_05}}
    ]
    constraintWeights [1,1,1,1,1]
   }
   n12 = "Constraint - dyn_necklace_bk_base_r (PPPPP)"!(
    ComponentPose=@n11
    Alpha=1
    ConstraintWeights_0=0.3605
    ConstraintWeights_1:float=0.4027
    ConstraintWeights_2:float=0.0196
    ConstraintWeights_3:float=0.0938
    ConstraintWeights_4:float=0.1234
   ) -> (Pose) {
    alphaBoolBlend {blendOption=Linear}
    boneToModify {boneName=dyn_necklace_bk_base_r}
    constraintSetup [
     {targetBone {boneName=neck_01}}
     {targetBone {boneName=neck_02}}
     {targetBone {boneName=head}}
     {targetBone {boneName=clavicle_r}}
     {targetBone {boneName=spine_05}}
    ]
    constraintWeights [1,1,1,1,1]
   }
   n13 = Get shoulderUp_L() -> (shoulderUp_L:double)
   n14 = "Transform (Modify) Bone - Bone: dyn_necklace_fr_base_l"(ComponentPose=@n12, Alpha=@n13, Translation="0,0,1.5", Rotation="0,0,0", Scale="1,1,1") {
    rotationMode=BMM_Additive
    rotationSpace=BCS_BoneSpace
    translationMode=BMM_Additive
    translationSpace=BCS_BoneSpace
    alphaBoolBlend {blendOption=Linear}
    alphaScaleBiasClamp {bMapRange=true, inRange {max=45}}
    boneToModify {boneName=dyn_necklace_fr_base_l}
    translation {z=1.5}
   }
   n15 = "Transform (Modify) Bone - Bone: dyn_necklace_side_base_l"(ComponentPose=@n14, Alpha=@n13, Translation="1,0,1.5", Rotation="15,0,0", Scale="1,1,1") {
    rotationMode=BMM_Additive
    rotationSpace=BCS_BoneSpace
    translationMode=BMM_Additive
    translationSpace=BCS_BoneSpace
    alphaBoolBlend {blendOption=Linear}
    alphaScaleBiasClamp {bMapRange=true, inRange {max=45}}
    boneToModify {boneName=dyn_necklace_side_base_l}
    rotation {pitch=15}
    translation {x=1, z=1.5}
   }
   n16 = "Transform (Modify) Bone - Bone: dyn_necklace_bk_base_l"(ComponentPose=@n15, Alpha=@n13, Translation="0.5,0,0", Rotation="0,0,0", Scale="1,1,1") {
    rotationMode=BMM_Additive
    rotationSpace=BCS_BoneSpace
    translationMode=BMM_Additive
    translationSpace=BCS_BoneSpace
    alphaBoolBlend {blendOption=Linear}
    alphaScaleBiasClamp {bMapRange=true, inRange {max=45}}
    boneToModify {boneName=dyn_necklace_bk_base_l}
    translation {x=0.5}
   }
   n17 = "Transform (Modify) Bone - Bone: dyn_necklace_side_base_r"(ComponentPose=@n16, Alpha=Get shoulderUp_R() -> (shoulderUp_R:double), Translation="-0.5,0,0", Rotation="0,0,0", Scale="1,1,1") {
    rotationMode=BMM_Additive
    rotationSpace=BCS_BoneSpace
    translationMode=BMM_Additive
    translationSpace=BCS_BoneSpace
    alphaBoolBlend {blendOption=Linear}
    alphaScaleBiasClamp {bMapRange=true, inRange {max=45}}
    boneToModify {boneName=dyn_necklace_side_base_r}
    translation {x=-0.5}
   }
   n18 = "Constraint - dyn_spike_base (PPPPPPP)"!(
    ComponentPose=@n17
    Alpha=1
    ConstraintWeights_0=0.0001
    ConstraintWeights_1:float=0.054
    ConstraintWeights_2:float=0.918
    ConstraintWeights_3:float=0
    ConstraintWeights_4:float=0
    ConstraintWeights_5:float=0
    ConstraintWeights_6:float=0.0279
   ) -> (Pose) {
    alphaBoolBlend {blendOption=Linear}
    boneToModify {boneName=dyn_spike_base}
    constraintSetup [
     {targetBone {boneName=dyn_necklace_side_l}}
     {targetBone {boneName=dyn_necklace_fr_l}}
     {targetBone {boneName=dyn_necklace_fr}}
     {targetBone {boneName=dyn_necklace_bk}}
     {targetBone {boneName=dyn_necklace_bk_r}}
     {targetBone {boneName=dyn_necklace_bk_l}}
     {targetBone {boneName=dyn_necklace_fr_r}}
    ]
    constraintWeights [1,1,1,1,1,1,1]
   }
   n19 = "Constraint - dyn_spike_base_l (PPPPPPP)"!(
    ComponentPose=@n18
    Alpha=1
    ConstraintWeights_0=0.0947
    ConstraintWeights_1:float=0.7192
    ConstraintWeights_2:float=0.1858
    ConstraintWeights_3:float=0
    ConstraintWeights_4:float=0
    ConstraintWeights_5:float=0.0002
    ConstraintWeights_6:float=0
   ) -> (Pose) {
    alphaBoolBlend {blendOption=Linear}
    boneToModify {boneName=dyn_spike_base_l}
    constraintSetup [
     {targetBone {boneName=dyn_necklace_side_l}}
     {targetBone {boneName=dyn_necklace_fr_l}}
     {targetBone {boneName=dyn_necklace_fr}}
     {targetBone {boneName=dyn_necklace_bk}}
     {targetBone {boneName=dyn_necklace_bk_r}}
     {targetBone {boneName=dyn_necklace_bk_l}}
     {targetBone {boneName=dyn_necklace_fr_r}}
    ]
    constraintWeights [1,1,1,1,1,1,1]
   }
   n20 = "Constraint - dyn_spike_base_r (PPPPPP)"!(
    ComponentPose=@n19
    Alpha=1
    ConstraintWeights_0=0.0002
    ConstraintWeights_1:float=0.3075
    ConstraintWeights_2:float=0
    ConstraintWeights_3:float=0.0001
    ConstraintWeights_4:float=0.6338
    ConstraintWeights_5:float=0.0584
   ) -> (Pose) {
    alphaBoolBlend {blendOption=Linear}
    boneToModify {boneName=dyn_spike_base_r}
    constraintSetup [
     {targetBone {boneName=dyn_necklace_fr_l}}
     {targetBone {boneName=dyn_necklace_fr}}
     {targetBone {boneName=dyn_necklace_bk}}
     {targetBone {boneName=dyn_necklace_bk_r}}
     {targetBone {boneName=dyn_necklace_fr_r}}
     {targetBone {boneName=dyn_necklace_side_r}}
    ]
    constraintWeights [1,1,1,1,1,1]
   }
   n21 = "Constraint - dyn_zipper_a_base (PPPPP)"!(
    ComponentPose=@n20
    Alpha=1
    ConstraintWeights_0=0.3776
    ConstraintWeights_1:float=0.0353
    ConstraintWeights_2:float=0.0971
    ConstraintWeights_3:float=0.2301
    ConstraintWeights_4:float=0.2599
   ) -> (Pose) {
    alphaBoolBlend {blendOption=Linear}
    boneToModify {boneName=dyn_zipper_a_base}
    constraintSetup [
     {targetBone {boneName=clavicle_r}}
     {targetBone {boneName=spine_02}}
     {targetBone {boneName=spine_03}}
     {targetBone {boneName=spine_04}}
     {targetBone {boneName=spine_05}}
    ]
    constraintWeights [1,1,1,1,1]
   }
   n22 = "Constraint - dyn_zipper_b_base (PPPPPPP)"!(
    ComponentPose=@n21
    Alpha=1
    ConstraintWeights_0=0.0114
    ConstraintWeights_1:float=0.0175
    ConstraintWeights_2:float=0.1079
    ConstraintWeights_3:float=0.2314
    ConstraintWeights_4:float=0.348
    ConstraintWeights_5:float=0.1325
    ConstraintWeights_6:float=0.1514
   ) -> (Pose) {
    alphaBoolBlend {blendOption=Linear}
    boneToModify {boneName=dyn_zipper_b_base}
    constraintSetup [
     {targetBone {boneName=clavicle_l}}
     {targetBone {boneName=spine_01}}
     {targetBone {boneName=spine_02}}
     {targetBone {boneName=spine_03}}
     {targetBone {boneName=spine_04}}
     {targetBone {boneName=spine_05}}
     {targetBone {boneName=clavicle_r}}
    ]
    constraintWeights [1,1,1,1,1,1,1]
   }
   n23 = "Constraint - dyn_chest_bag_a_base (PPPPPPPP)"!(
    ComponentPose=@n22
    Alpha=1
    ConstraintWeights_0=0.0177
    ConstraintWeights_1:float=0.0552
    ConstraintWeights_2:float=0.1792
    ConstraintWeights_3:float=0.3547
    ConstraintWeights_4:float=0.2613
    ConstraintWeights_5:float=0.0457
    ConstraintWeights_6:float=0.0705
    ConstraintWeights_7:float=0.0158
   ) -> (Pose) {
    alphaBoolBlend {blendOption=Linear}
    boneToModify {boneName=dyn_chest_bag_a_base}
    constraintSetup [
     {targetBone {boneName=clavicle_l}}
     {targetBone {boneName=spine_01}}
     {targetBone {boneName=spine_02}}
     {targetBone {boneName=spine_03}}
     {targetBone {boneName=spine_04}}
     {targetBone {boneName=spine_05}}
     {targetBone {boneName=clavicle_r}}
     {targetBone {boneName=pelvis}}
    ]
    constraintWeights [1,1,1,1,1,1,1,1]
   }
   n24 = "Constraint - dyn_chest_bag_b_base (PPPPPPP)"!(
    ComponentPose=@n23
    Alpha=1
    ConstraintWeights_0=0.0661
    ConstraintWeights_1:float=0.0852
    ConstraintWeights_2:float=0.2152
    ConstraintWeights_3:float=0.3853
    ConstraintWeights_4:float=0.1949
    ConstraintWeights_5:float=0.0132
    ConstraintWeights_6:float=0.0402
   ) -> (Pose) {
    alphaBoolBlend {blendOption=Linear}
    boneToModify {boneName=dyn_chest_bag_b_base}
    constraintSetup [
     {targetBone {boneName=clavicle_l}}
     {targetBone {boneName=spine_01}}
     {targetBone {boneName=spine_02}}
     {targetBone {boneName=spine_03}}
     {targetBone {boneName=spine_04}}
     {targetBone {boneName=spine_05}}
     {targetBone {boneName=pelvis}}
    ]
    constraintWeights [1,1,1,1,1,1,1]
   }
   n25 = Get legUp_L()
   n26 = "Transform (Modify) Bone - Bone: dyn_pouch_a_base_l"(ComponentPose=@n24, Alpha=@n25, Translation="0,-3.5,-0.6", Rotation="0,-10,-25", Scale="1,1,1") {
    rotationMode=BMM_Additive
    rotationSpace=BCS_BoneSpace
    translationMode=BMM_Additive
    translationSpace=BCS_BoneSpace
    alphaBoolBlend {blendOption=Linear}
    alphaScaleBiasClamp {bMapRange=true, inRange {max=0, min=-100}, outRange {max=0, min=1}}
    boneToModify {boneName=dyn_pouch_a_base_l}
    rotation {roll=-25, yaw=-10}
    translation {y=-3.5, z=-0.6}
   }
   n27 = "Transform (Modify) Bone - Bone: dyn_pouch_b_base_l"(ComponentPose=@n26, Alpha=@n25, Translation="0,-2.5,0", Rotation="0,-5,-15", Scale="1,1,1") {
    rotationMode=BMM_Additive
    rotationSpace=BCS_BoneSpace
    translationMode=BMM_Additive
    translationSpace=BCS_BoneSpace
    alphaBoolBlend {blendOption=Linear}
    alphaScaleBiasClamp {bMapRange=true, inRange {max=0, min=-100}, outRange {max=0, min=1}}
    boneToModify {boneName=dyn_pouch_b_base_l}
    rotation {roll=-15, yaw=-5}
    translation {y=-2.5}
   }
   n28 = "Transform (Modify) Bone - Bone: dyn_pouch_c_base_l"(ComponentPose=@n27, Alpha=@n25, Translation="0,-1.5,0", Rotation="0,-3,-5", Scale="1,1,1") {
    rotationMode=BMM_Additive
    rotationSpace=BCS_BoneSpace
    translationMode=BMM_Additive
    translationSpace=BCS_BoneSpace
    alphaBoolBlend {blendOption=Linear}
    alphaScaleBiasClamp {bMapRange=true, inRange {max=0, min=-100}, outRange {max=0, min=1}}
    boneToModify {boneName=dyn_pouch_c_base_l}
    rotation {roll=-5, yaw=-3}
    translation {y=-1.5}
   }
   n29 = Get legUp_R()
   n30 = "Transform (Modify) Bone - Bone: dyn_pouch_a_base_r"(ComponentPose=@n28, Alpha=@n29, Translation="0,3.5,0.6", Rotation="0,-10,-25", Scale="1,1,1") {
    rotationMode=BMM_Additive
    rotationSpace=BCS_BoneSpace
    translationMode=BMM_Additive
    translationSpace=BCS_BoneSpace
    alphaBoolBlend {blendOption=Linear}
    alphaScaleBiasClamp {bMapRange=true, inRange {max=0, min=-100}, outRange {max=0, min=1}}
    boneToModify {boneName=dyn_pouch_a_base_r}
    rotation {roll=-25, yaw=-10}
    translation {y=3.5, z=0.6}
   }
   n31 = "Transform (Modify) Bone - Bone: dyn_pouch_b_base_r"(ComponentPose=@n30, Alpha=@n29, Translation="0,2.5,0", Rotation="0,-5,-15", Scale="1,1,1") {
    rotationMode=BMM_Additive
    rotationSpace=BCS_BoneSpace
    translationMode=BMM_Additive
    translationSpace=BCS_BoneSpace
    alphaBoolBlend {blendOption=Linear}
    alphaScaleBiasClamp {bMapRange=true, inRange {max=0, min=-100}, outRange {max=0, min=1}}
    boneToModify {boneName=dyn_pouch_b_base_r}
    rotation {roll=-15, yaw=-5}
    translation {y=2.5}
   }
   n32 = "Transform (Modify) Bone - Bone: dyn_pouch_c_base_r"(ComponentPose=@n31, Alpha=@n29, Translation="0,1.5,0", Rotation="0,-3,-5", Scale="1,1,1") {
    rotationMode=BMM_Additive
    rotationSpace=BCS_BoneSpace
    translationMode=BMM_Additive
    translationSpace=BCS_BoneSpace
    alphaBoolBlend {blendOption=Linear}
    alphaScaleBiasClamp {bMapRange=true, inRange {max=0, min=-100}, outRange {max=0, min=1}}
    boneToModify {boneName=dyn_pouch_c_base_r}
    rotation {roll=-5, yaw=-3}
    translation {y=1.5}
   }
   n33 = Get kneeBend_L() -> (kneeBend_L:double)
   n34 = "Transform (Modify) Bone - Bone: dyn_pants_fr_base_l"(ComponentPose=@n32, Alpha=@n33, Translation="-3,0,0", Rotation="0,-5,0", Scale="1,1,1") {
    rotationMode=BMM_Additive
    rotationSpace=BCS_BoneSpace
    translationMode=BMM_Additive
    translationSpace=BCS_BoneSpace
    alphaBoolBlend {blendOption=Linear}
    alphaScaleBiasClamp {bMapRange=true, inRange {max=115, min=40}}
    boneToModify {boneName=dyn_pants_fr_base_l}
    rotation {yaw=-5}
    translation {x=-3}
   }
   n35 = "Transform (Modify) Bone - Bone: dyn_pants_bk_base_l"(ComponentPose=@n34, Alpha=@n33, Translation="-15,0,0", Rotation="0,10,0", Scale="1,1,1") {
    rotationMode=BMM_Additive
    rotationSpace=BCS_BoneSpace
    translationMode=BMM_Additive
    translationSpace=BCS_BoneSpace
    alphaBoolBlend {blendOption=Linear}
    alphaScaleBiasClamp {bMapRange=true, inRange {max=115, min=40}}
    boneToModify {boneName=dyn_pants_bk_base_l}
    rotation {yaw=10}
    translation {x=-15}
   }
   n36 = Get kneeBend_R() -> (kneeBend_R:double)
   n37 = "Transform (Modify) Bone - Bone: dyn_pants_fr_base_r"(ComponentPose=@n35, Alpha=@n36, Translation="3,0,0", Rotation="0,-5,0", Scale="1,1,1") {
    rotationMode=BMM_Additive
    rotationSpace=BCS_BoneSpace
    translationMode=BMM_Additive
    translationSpace=BCS_BoneSpace
    alphaBoolBlend {blendOption=Linear}
    alphaScaleBiasClamp {bMapRange=true, inRange {max=115, min=40}}
    boneToModify {boneName=dyn_pants_fr_base_r}
    rotation {yaw=-5}
    translation {x=3}
   }
   n38 = "Transform (Modify) Bone - Bone: dyn_pants_bk_base_r"(ComponentPose=@n37, Alpha=@n36, Translation="15,0,0", Rotation="0,10,0", Scale="1,1,1") {
    rotationMode=BMM_Additive
    rotationSpace=BCS_BoneSpace
    translationMode=BMM_Additive
    translationSpace=BCS_BoneSpace
    alphaBoolBlend {blendOption=Linear}
    alphaScaleBiasClamp {bMapRange=true, inRange {max=115, min=40}}
    boneToModify {boneName=dyn_pants_bk_base_r}
    rotation {yaw=10}
    translation {x=15}
   }
   n39 = Anim Dynamics - Bone: dyn_zipper_a(ComponentPose=@n38, Alpha=0.8, ExternalForce="0,0,0") {
    angularDampingOverride=1
    angularSpringConstant=1500
    bAngularSpring=true
    bGravityOverrideInSimSpace=true
    bOverrideAngularDamping=true
    gravityScale=0.35
    simulationSpace=RootRelative
    alphaBoolBlend {blendOption=Linear}
    boundBone {boneName=dyn_zipper_a}
    physicsBodyDefinitions [{
      sphereCollisionRadius=0
      boundBone {boneName=dyn_zipper_a}
      boxExtents=<10 20 20>
      constraintSetup {angularLimitsMax=<5 20 60>, angularLimitsMin=<-5 -20 0>, angularTarget=<0 0 0>}
      localJointOffset=<10 0 0>
     }]
   }
   n40 = Anim Dynamics - Bone: dyn_zipper_b(ComponentPose=@n39, Alpha=1, ExternalForce="0,0,0") {
    angularDampingOverride=1
    angularSpringConstant=1500
    bAngularSpring=true
    bGravityOverrideInSimSpace=true
    bOverrideAngularDamping=true
    gravityScale=0.35
    simulationSpace=RootRelative
    alphaBoolBlend {blendOption=Linear}
    boundBone {boneName=dyn_zipper_b}
    physicsBodyDefinitions [{
      sphereCollisionRadius=0
      boundBone {boneName=dyn_zipper_b}
      boxExtents=<10 10 10>
      constraintSetup {angularLimitsMax=<10 30 60>, angularLimitsMin=<0 -30 0>, angularTarget=<0 0 0>}
      localJointOffset=<5 0 0>
     }]
   }
   n41 = RigidBody(ComponentPose=@n40) {
    bClampLinearTranslationLimitToRefPose=true
    bOverrideWorldGravity=true
    overridePhysicsAsset=&/M_MED_Floatilla_Captain_Physics
    simulationSpace=BaseBoneSpace
    alphaBoolBlend {blendOption=Linear}
    baseBoneRef {boneName=pelvis}
    componentLinearAccScale=<0.35 0.35 0.35>
    componentLinearVelScale=<0.35 0.35 0.35>
    overrideWorldGravity {z=-980}
   }
   n42 = RigidBody(ComponentPose=@n41, Alpha=Get pantsRigidAlpha() -> (pantsRigidAlpha:double)) {
    bClampLinearTranslationLimitToRefPose=true
    bOverrideWorldGravity=true
    overridePhysicsAsset=&/M_MED_Floatilla_Captain_Pants_Physics
    simulationSpace=BaseBoneSpace
    alphaBoolBlend {blendOption=Linear}
    baseBoneRef {boneName=root}
    componentLinearAccScale=<0.25 0.25 0.25>
    componentLinearVelScale=<0.25 0.25 0.25>
    overrideWorldGravity {z=-980}
   }
   Output Pose(Result=Component To Local(ComponentPose=@n42)) {layerGroup=None}
  ]}
 EventGraph {nodes [
   Event BlueprintUpdateAnimation()
   > n1 = Sequence!(execute) -> (then_0, then_1, then_2:Exec, then_3:Exec, then_4:Exec, then_5:Exec)
   > n2 = Sequence!(execute) -> (then_0, then_1)
   n3 = GetDeltaTransformFromRefPose(self=GetOwnerSkeletalMeshComponent(), BoneName=clavicle_l, BaseName=spine_05)
   > Set shoulderUp_L(shoulderUp_L:double=BreakRotator(InRot=BreakTransform(InTransform=@n3).Rotation).Pitch) -> (Output_Get:double)
   n4 = GetDeltaTransformFromRefPose(self=GetOwnerSkeletalMeshComponent(), BoneName=clavicle_l, BaseName=spine_05)
   Set shoulderUp_R(execute=@n2.then_1, shoulderUp_R:double=BreakRotator(InRot=BreakTransform(InTransform=@n4).Rotation).Pitch) -> (Output_Get:double)
   n5 = Sequence(execute=@n1.then_1)
   n6 = GetDeltaTransformFromRefPose(self=GetOwnerSkeletalMeshComponent(), BoneName=thigh_r, BaseName=pelvis)
   Set legUp_R(execute=@n5.then_1, legUp_R:double=BreakRotator(InRot=BreakTransform(InTransform=@n6).Rotation).Yaw) -> (Output_Get:double)
   n7 = GetDeltaTransformFromRefPose(self=GetOwnerSkeletalMeshComponent(), BoneName=thigh_l, BaseName=pelvis)
   Set legUp_L(execute=@n5.then_0, legUp_L:double=BreakRotator(InRot=BreakTransform(InTransform=@n7).Rotation).Yaw) -> (Output_Get:double)
   n8 = Sequence(execute=@n1.then_2)
   n9 = GetDeltaTransformFromRefPose(self=GetOwnerSkeletalMeshComponent(), BoneName=calf_l, BaseName=thigh_l)
   > Set kneeBend_L(kneeBend_L:double=BreakRotator(InRot=BreakTransform(InTransform=@n9).Rotation).Yaw) -> (Output_Get:double)
   n10 = GetDeltaTransformFromRefPose(self=GetOwnerSkeletalMeshComponent(), BoneName=calf_r, BaseName=thigh_r)
   Set kneeBend_R(execute=@n8.then_1, kneeBend_R:double=BreakRotator(InRot=BreakTransform(InTransform=@n10).Rotation).Yaw) -> (Output_Get:double)
   n11 = Branch(execute=@n1.then_3, Condition=Get bIsCrouching())
   > Set pantsRigidAlpha(pantsRigidAlpha:double=0) -> (Output_Get:double)
   n12 = Branch(execute=@n11.else, Condition=OR Boolean(A=Get bIsJumping(), B=Get bIsFalling()))
   > Set pantsRigidAlpha(pantsRigidAlpha=0.1)
   Set pantsRigidAlpha(execute=@n12.else, pantsRigidAlpha=1)
   TryGetPawnOwner()
  ]}
]
variables [
 kneeBend_L {type=number}
 kneeBend_R {type=number}
 legUp_L {type=number}
 legUp_R {type=number}
 pantsRigidAlpha {type=number}
 shoulderUp_L {type=number}
 shoulderUp_R {type=number}
]
