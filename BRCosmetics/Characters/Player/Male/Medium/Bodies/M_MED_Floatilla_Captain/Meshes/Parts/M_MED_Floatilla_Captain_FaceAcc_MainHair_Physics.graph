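# Physics asset for the main hair part of the male medium Floatilla Captain player cosmetic, giving its hair bones capsule bodies for physics.
# A three-segment tail chain hangs from a base body on the head, with two three-segment strands, F and G, branching from the first tail segment.
# Hair joints use limited swing and twist, tightest at strand tips.
# Spine, neck and head capsules form the body volume, linked by free-rotation constraints, as is the tail base.
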PhysicsAsset /BRCosmetics/Characters/Player/Male/Medium/Bodies/M_MED_Floatilla_Captain/Meshes/Parts/M_MED_Floatilla_Captain_FaceAcc_MainHair_Physics
bodies {
 dyn_hair_F_a [{shapeType=Capsule, length=2, radius=1.25, center=<2.1707 0 0>, rotation=<rot 90 0 0>}]
 dyn_hair_F_b [{shapeType=Capsule, length=3.25, radius=1.5, center=<3.0935 0 0>, rotation=<rot 90 0 0>}]
 dyn_hair_F_c [{shapeType=Capsule, length=2.75, radius=1.5, center=<2.5456 0 0.0001>, rotation=<rot 90 0 0>}]
 dyn_hair_G_a [{shapeType=Capsule, length=3.25, radius=1.25, center=<2.6922 0 0>, rotation=<rot 90 0 0>}]
 dyn_hair_G_b [{shapeType=Capsule, length=1.5, radius=1, center=<1.5869 -0.2599 0>, rotation=<rot 90 0 9.3014>}]
 dyn_hair_G_c [{shapeType=Capsule, length=2, radius=1.5, center=<2.4664 -0.2958 0>, rotation=<rot 90 6.831 9.9976>}]
 dyn_hair_tail_a [{shapeType=Capsule, length=2.5, radius=3.25, center=<4.6406 0 -0.0001>, rotation=<rot 90 0 0>}]
 dyn_hair_tail_a_base [{shapeType=Capsule, length=1, radius=1}]
 dyn_hair_tail_b [{shapeType=Capsule, length=1, radius=3.5, center=<3.8654 -0.3274 -0.0001>, rotation=<rot 90 0 0>}]
 dyn_hair_tail_c [{shapeType=Capsule, length=4.25, radius=3.25, center=<6.0649 0 0>, rotation=<rot 90 0 0>}]
 head [{shapeType=Capsule, length=1.5, radius=9.5, center=<5.9146 2.5338 0>, rotation=<rot 90 0 0>}]
 neck_01 [{shapeType=Capsule, length=1.5, radius=6.75, center=<-0.7613 0.5004 0>, rotation=<rot 90 0 -12.6134>}]
 neck_02 [{shapeType=Capsule, length=1.5, radius=6, center=<0 0.838 0>, rotation=<rot 90 0 0>}]
 spine_04 [{shapeType=Capsule, length=6.5, radius=13.25, center=<4.2095 -0.5302 0>}]
 spine_05 [{shapeType=Capsule, length=21.5, radius=12.5, center=<0.6848 -0.6617 0>}]
}
constraints [
 {
  bone1Name=dyn_hair_tail_a_base
  bone2Name=head
  swing1LimitDegrees=45
  swing1Motion=Free
  swing2LimitDegrees=45
  swing2Motion=Free
  twistLimitDegrees=45
  twistMotion=Free
 }
 {bone1Name=dyn_hair_tail_a, bone2Name=dyn_hair_tail_a_base, swing1LimitDegrees=15, swing2LimitDegrees=15, twistLimitDegrees=5}
 {bone1Name=dyn_hair_tail_b, bone2Name=dyn_hair_tail_a, swing1LimitDegrees=45, swing2LimitDegrees=45, twistLimitDegrees=5}
 {bone1Name=dyn_hair_tail_c, bone2Name=dyn_hair_tail_b, swing1LimitDegrees=25, swing2LimitDegrees=25, twistLimitDegrees=5}
 {
  bone1Name=head
  bone2Name=neck_02
  swing1LimitDegrees=45
  swing1Motion=Free
  swing2LimitDegrees=45
  swing2Motion=Free
  twistLimitDegrees=45
  twistMotion=Free
 }
 {
  bone1Name=spine_05
  bone2Name=spine_04
  swing1LimitDegrees=45
  swing1Motion=Free
  swing2LimitDegrees=45
  swing2Motion=Free
  twistLimitDegrees=45
  twistMotion=Free
 }
 {
  bone1Name=neck_01
  bone2Name=spine_05
  swing1LimitDegrees=45
  swing1Motion=Free
  swing2LimitDegrees=45
  swing2Motion=Free
  twistLimitDegrees=45
  twistMotion=Free
 }
 {bone1Name=dyn_hair_G_a, bone2Name=dyn_hair_tail_a, swing1LimitDegrees=10, swing2LimitDegrees=10, twistLimitDegrees=5}
 {bone1Name=dyn_hair_G_b, bone2Name=dyn_hair_G_a, swing1LimitDegrees=10, swing2LimitDegrees=10, twistLimitDegrees=5}
 {bone1Name=dyn_hair_G_c, bone2Name=dyn_hair_G_b, swing1LimitDegrees=2, swing2LimitDegrees=2, twistLimitDegrees=2}
 {bone1Name=dyn_hair_F_a, bone2Name=dyn_hair_tail_a, swing1LimitDegrees=10, swing2LimitDegrees=10, twistLimitDegrees=5}
 {bone1Name=dyn_hair_F_b, bone2Name=dyn_hair_F_a, swing1LimitDegrees=10, swing2LimitDegrees=10, twistLimitDegrees=5}
 {bone1Name=dyn_hair_F_c, bone2Name=dyn_hair_F_b, swing1LimitDegrees=2, swing2LimitDegrees=2, twistLimitDegrees=2}
 {
  bone1Name=neck_02
  bone2Name=neck_01
  swing1LimitDegrees=45
  swing1Motion=Free
  swing2LimitDegrees=45
  swing2Motion=Free
  twistLimitDegrees=45
  twistMotion=Free
 }
]
