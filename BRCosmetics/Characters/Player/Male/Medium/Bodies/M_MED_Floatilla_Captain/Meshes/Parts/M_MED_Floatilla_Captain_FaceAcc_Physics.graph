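# Physics asset for the Floatilla Captain's face accessory mesh, setting up capsule bodies and joints for its dynamic hair bones.
# A tail chain hangs from the head through a base body, with lettered strands A–H branching off its first segment and five tips fanning out from its last.
# Hair joints use tight swing and twist limits.
# Head, neck and upper spine capsules, linked by free joints, provide body volume around the hair.
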PhysicsAsset /BRCosmetics/Characters/Player/Male/Medium/Bodies/M_MED_Floatilla_Captain/Meshes/Parts/M_MED_Floatilla_Captain_FaceAcc_Physics
bodies {
 dyn_hair_A_a [{shapeType=Capsule, length=3.1517, radius=1.648, center=<-0.2002 -2.8117 0.2429>, rotation=<rot 0 -10.6553 -83.9214>}]
 dyn_hair_B_a [{shapeType=Capsule, length=1.5, radius=1.5, center=<1.9102 0 0>, rotation=<rot 90 0 0>}]
 dyn_hair_B_b [{shapeType=Capsule, length=1.5, radius=1.5, center=<2.0133 0 0>, rotation=<rot 90 0 0>}]
 dyn_hair_C_a [{shapeType=Capsule, length=1.6169, radius=1.355, center=<2.1732 -0.2687 -0.034>, rotation=<rot 0 -94.1392 -88.1541>}]
 dyn_hair_C_b [{shapeType=Capsule, length=1.6205, radius=1.3788, center=<2.1482 0.0666 0.0884>, rotation=<rot 0 93.3801 -85.6343>}]
 dyn_hair_C_c [{shapeType=Capsule, length=3.2499, radius=1.5133, center=<2.9542 -0.2159 0.2178>, rotation=<rot 0 -97.8913 -87.591>}]
 dyn_hair_D_a [{shapeType=Capsule, length=1.5879, radius=1.3035, center=<2.1545 0.1407 0.0656>, rotation=<rot 0 -84.2868 -89.9593>}]
 dyn_hair_D_b [{shapeType=Capsule, length=1.5495, radius=1.2873, center=<2.1063 -0.1524 0.1961>, rotation=<rot 0 93.2437 -87.2978>}]
 dyn_hair_D_c [{shapeType=Capsule, length=3.554, radius=1.485, center=<2.7792 -0.1358 -0.5049>, rotation=<rot 0 -104.4856 -77.2682>}]
 dyn_hair_E_a [{shapeType=Capsule, length=3.75, radius=1.25, center=<2.9729 0 0>, rotation=<rot 90 0 0>}]
 dyn_hair_E_b [{shapeType=Capsule, length=3, radius=1.25, center=<2.7258 0 -0.1183>, rotation=<rot 86.6879 0.0003 0.0003>}]
 dyn_hair_F_a [{shapeType=Capsule, length=2, radius=1.25, center=<2.1707 0 0>, rotation=<rot 90 0 0>}]
 dyn_hair_F_b [{shapeType=Capsule, length=3.5, radius=1.25, center=<3.0935 0 0>, rotation=<rot 90 0 0>}]
 dyn_hair_F_c [{shapeType=Capsule, length=2.25, radius=1.5, center=<2.647 0.1008 0.1677>, rotation=<rot 88.1607 -179.9999 -179.9999>}]
 dyn_hair_G_a [{shapeType=Capsule, length=3.25, radius=1.25, center=<2.6922 0 0>, rotation=<rot 90 0 0>}]
 dyn_hair_G_b [{shapeType=Capsule, length=1.5, radius=1, center=<1.5869 -0.2599 0>, rotation=<rot 90 0 9.3014>}]
 dyn_hair_G_c [{shapeType=Capsule, length=2.5, radius=1.25, center=<2.3568 -0.2383 0.0001>, rotation=<rot 90 0 -6.2755>}]
 dyn_hair_H_a [{shapeType=Capsule, length=2.4325, radius=1.2496, center=<2.2957 -0.2879 -0.0119>, rotation=<rot 0 -98.2374 -88.9261>}]
 dyn_hair_H_b [{shapeType=Capsule, length=3.7008, radius=1.2821, center=<3.1666 0.0363 -0.0435>, rotation=<rot 0 -88.7518 -88.8105>}]
 dyn_hair_H_c [{shapeType=Capsule, length=3.9907, radius=1.589, center=<4.4222 -0.1866 0.4438>, rotation=<rot 0.5971 84.7482 -81.3543>}]
 dyn_hair_tail_a [{shapeType=Capsule, length=2.5, radius=3.25, center=<4.6406 0 -0.0001>, rotation=<rot 90 0 0>}]
 dyn_hair_tail_a_base [{shapeType=Capsule, length=1, radius=1}]
 dyn_hair_tail_b [{shapeType=Capsule, length=1, radius=3.5, center=<3.8654 -0.3274 -0.0001>, rotation=<rot 90 0 0>}]
 dyn_hair_tail_c [{shapeType=Capsule, length=4.25, radius=3.25, center=<6.0649 0 0>, rotation=<rot 90 0 0>}]
 dyn_hair_tip_a [{shapeType=Capsule, length=4, radius=1.5, center=<2.8872 -0.4548 -0.0006>, rotation=<rot 83.2596 -3.76 -0.0002>}]
 dyn_hair_tip_b [{shapeType=Capsule, length=3.5, radius=1.25, center=<2.839 0 0.3602>, rotation=<rot 85.5141 -0.0002 -0.0002>}]
 dyn_hair_tip_c [{shapeType=Capsule, length=3.5, radius=1.5, center=<3.2187 0.0288 0.6001>, rotation=<rot 80.37 -179.9998 -179.1193>}]
 dyn_hair_tip_d [{shapeType=Capsule, length=3, radius=1.25, center=<2.9081 0.2458 0.1728>, rotation=<rot 90 0 0>}]
 dyn_hair_tip_e [{shapeType=Capsule, length=3.75, radius=1.25, center=<3.684 -0.55 -0.7519>, rotation=<rot 90 180 -171.509>}]
 head [{shapeType=Capsule, length=1.5, radius=9.5, center=<5.9146 2.5338 0>, rotation=<rot 90 0 0>}]
 neck_01 [{shapeType=Capsule, length=1.5, radius=6.75, center=<-0.7613 0.5004 0>, rotation=<rot 90 0 -12.6134>}]
 neck_02 [{shapeType=Capsule, length=1.5, radius=6, center=<0 0.838 0>, rotation=<rot 90 0 0>}]
 spine_04 [{shapeType=Capsule, length=6.5, radius=13.25, center=<4.2095 -0.5302 0>}]
 spine_05 [{shapeType=Capsule, length=21.5, radius=12.5, center=<0.6848 -0.6617 0>}]
}
constraints [
 {
  bone1Name=dyn_hair_tail_a_base
  bone2Name=head
  swing1LimitDegrees=45
  swing1Motion=Free
  swing2LimitDegrees=45
  swing2Motion=Free
  twistLimitDegrees=45
  twistMotion=Free
 }
 {bone1Name=dyn_hair_tail_a, bone2Name=dyn_hair_tail_a_base, swing1LimitDegrees=15, swing2LimitDegrees=15, twistLimitDegrees=5}
 {bone1Name=dyn_hair_tail_b, bone2Name=dyn_hair_tail_a, swing1LimitDegrees=45, swing2LimitDegrees=45, twistLimitDegrees=5}
 {bone1Name=dyn_hair_tail_c, bone2Name=dyn_hair_tail_b, swing1LimitDegrees=25, swing2LimitDegrees=25, twistLimitDegrees=5}
 {
  bone1Name=head
  bone2Name=neck_02
  swing1LimitDegrees=45
  swing1Motion=Free
  swing2LimitDegrees=45
  swing2Motion=Free
  twistLimitDegrees=45
  twistMotion=Free
 }
 {
  bone1Name=spine_05
  bone2Name=spine_04
  swing1LimitDegrees=45
  swing1Motion=Free
  swing2LimitDegrees=45
  swing2Motion=Free
  twistLimitDegrees=45
  twistMotion=Free
 }
 {
  bone1Name=neck_01
  bone2Name=spine_05
  swing1LimitDegrees=45
  swing1Motion=Free
  swing2LimitDegrees=45
  swing2Motion=Free
  twistLimitDegrees=45
  twistMotion=Free
 }
 {bone1Name=dyn_hair_tip_a, bone2Name=dyn_hair_tail_c, swing1LimitDegrees=25, swing2LimitDegrees=25, twistLimitDegrees=5}
 {bone1Name=dyn_hair_tip_b, bone2Name=dyn_hair_tail_c, swing1LimitDegrees=25, swing2LimitDegrees=25, twistLimitDegrees=5}
 {bone1Name=dyn_hair_tip_c, bone2Name=dyn_hair_tail_c, swing1LimitDegrees=25, swing2LimitDegrees=25, twistLimitDegrees=5}
 {bone1Name=dyn_hair_tip_d, bone2Name=dyn_hair_tail_c, swing1LimitDegrees=20, swing2LimitDegrees=20, twistLimitDegrees=5}
 {bone1Name=dyn_hair_tip_e, bone2Name=dyn_hair_tail_c, swing1LimitDegrees=25, swing2LimitDegrees=25, twistLimitDegrees=5}
 {bone1Name=dyn_hair_A_a, bone2Name=dyn_hair_tail_a, swing1LimitDegrees=30, swing2LimitDegrees=5, twistLimitDegrees=20}
 {bone1Name=dyn_hair_B_a, bone2Name=dyn_hair_tail_a, swing1LimitDegrees=15, swing2LimitDegrees=15, twistLimitDegrees=5}
 {bone1Name=dyn_hair_B_b, bone2Name=dyn_hair_B_a, swing1LimitDegrees=30, swing2LimitDegrees=30, twistLimitDegrees=5}
 {bone1Name=dyn_hair_E_a, bone2Name=dyn_hair_tail_a, swing1LimitDegrees=30, swing2LimitDegrees=30, twistLimitDegrees=5}
 {bone1Name=dyn_hair_E_b, bone2Name=dyn_hair_E_a, swing1LimitDegrees=30, swing2LimitDegrees=30, twistLimitDegrees=5}
 {bone1Name=dyn_hair_H_a, bone2Name=dyn_hair_tail_a, swing1LimitDegrees=15, swing2LimitDegrees=15, twistLimitDegrees=5}
 {bone1Name=dyn_hair_H_b, bone2Name=dyn_hair_H_a, swing1LimitDegrees=30, swing2LimitDegrees=30, twistLimitDegrees=5}
 {bone1Name=dyn_hair_H_c, bone2Name=dyn_hair_H_b, swing1LimitDegrees=30, swing2LimitDegrees=30, twistLimitDegrees=5}
 {bone1Name=dyn_hair_C_a, bone2Name=dyn_hair_tail_a, swing1LimitDegrees=15, swing1Motion=Locked, swing2LimitDegrees=15, twistLimitDegrees=5}
 {bone1Name=dyn_hair_C_b, bone2Name=dyn_hair_C_a, swing1LimitDegrees=30, swing2LimitDegrees=30, twistLimitDegrees=5}
 {bone1Name=dyn_hair_C_c, bone2Name=dyn_hair_C_b, swing1LimitDegrees=30, swing2LimitDegrees=30, twistLimitDegrees=5}
 {bone1Name=dyn_hair_D_a, bone2Name=dyn_hair_tail_a, swing1LimitDegrees=15, swing2LimitDegrees=15, twistLimitDegrees=5}
 {bone1Name=dyn_hair_D_b, bone2Name=dyn_hair_D_a, swing1LimitDegrees=5, swing2LimitDegrees=5, twistLimitDegrees=5}
 {bone1Name=dyn_hair_D_c, bone2Name=dyn_hair_D_b, swing1LimitDegrees=15, swing2LimitDegrees=15, twistLimitDegrees=5}
 {bone1Name=dyn_hair_G_a, bone2Name=dyn_hair_tail_a, swing1LimitDegrees=15, swing2LimitDegrees=15, twistLimitDegrees=5}
 {bone1Name=dyn_hair_G_b, bone2Name=dyn_hair_G_a, swing1LimitDegrees=15, swing2LimitDegrees=15, twistLimitDegrees=5}
 {bone1Name=dyn_hair_G_c, bone2Name=dyn_hair_G_b, swing1LimitDegrees=15, swing2LimitDegrees=15, twistLimitDegrees=5}
 {bone1Name=dyn_hair_F_a, bone2Name=dyn_hair_tail_a, swing1LimitDegrees=15, swing2LimitDegrees=15, twistLimitDegrees=5}
 {bone1Name=dyn_hair_F_b, bone2Name=dyn_hair_F_a, swing1LimitDegrees=15, swing2LimitDegrees=15, twistLimitDegrees=5}
 {bone1Name=dyn_hair_F_c, bone2Name=dyn_hair_F_b, swing1LimitDegrees=15, swing2LimitDegrees=15, twistLimitDegrees=5}
 {
  bone1Name=neck_02
  bone2Name=neck_01
  swing1LimitDegrees=45
  swing1Motion=Free
  swing2LimitDegrees=45
  swing2Motion=Free
  twistLimitDegrees=45
  twistMotion=Free
 }
]
